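# Post-process style AnimBlueprint for the F_MED_SlumpAfter_Awe female body cosmetic, adding secondary motion to its skirt flaps.
# It copies the attached parent's pose through a Control Rig, then stacks additive bone offsets on four flap bones.
# A pelvis-space RigidBody node simulates physics, after which those bones are shrunk along X.
# The EventGraph sets gravity and slide/vehicle alpha variables, with no visible reads in the AnimGraph.
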AnimBlueprint /BRCosmetics/Characters/Player/Female/Medium/Bodies/F_MED_SlumpAfter_Awe/Meshes/F_MED_SlumpAfter_Awe_AnimBP
target_skeleton=&/F_MED_SlumpAfter_Awe_Skeleton
graphs [
 AnimGraph {nodes [
   n1 = Control Rig(Source=Copy Pose From Mesh() {bUseAttachedParent=true}) {
    alphaBoolBlend {blendOption=Linear}
    controlRigAssetReference {controlRigAsset=&/F_MED_SlumpAfter_Awe_CtrlRig}
   }
   n2 = "Transform (Modify) Bone - Bone: dyn_skirtflap_l_MSTR"(ComponentPose=Local To Component(LocalPose=@n1), Alpha=1, Translation="0,0,0", Rotation="30,-40,30", Scale="1,1,1") {
    rotationMode=BMM_Additive
    rotationSpace=BCS_BoneSpace
    scaleMode=BMM_Replace
    scaleSpace=BCS_BoneSpace
    translationMode=BMM_Additive
    translationSpace=BCS_BoneSpace
    alphaBoolBlend {blendOption=Linear}
    alphaScaleBiasClamp {interpSpeedDecreasing=4}
    boneToModify {boneName=dyn_skirtflap_l_MSTR}
    rotation=<rot 30 -40 30>
   }
   n3 = "Transform (Modify) Bone - Bone: dyn_skirtflap_r_MSTR"(ComponentPose=@n2, Alpha=1, Translation="0,0,0", Rotation="30,40,-30", Scale="1,1,1") {
    rotationMode=BMM_Additive
    rotationSpace=BCS_BoneSpace
    scaleMode=BMM_Replace
    scaleSpace=BCS_BoneSpace
    translationMode=BMM_Additive
    translationSpace=BCS_BoneSpace
    alphaBoolBlend {blendOption=Linear}
    alphaScaleBiasClamp {interpSpeedDecreasing=4}
    boneToModify {boneName=dyn_skirtflap_r_MSTR}
    rotation=<rot 30 40 -30>
   }
   n4 = "Transform (Modify) Bone - Bone: dyn_skirtflap_bk_l_MSTR"(ComponentPose=@n3, Alpha=1, Translation="-10,0,5", Rotation="75,-40,30", Scale="1,1,1") {
    rotationMode=BMM_Additive
    rotationSpace=BCS_BoneSpace
    scaleMode=BMM_Replace
    scaleSpace=BCS_BoneSpace
    translationMode=BMM_Additive
    translationSpace=BCS_BoneSpace
    alphaBoolBlend {blendOption=Linear}
    alphaScaleBiasClamp {interpSpeedDecreasing=4}
    boneToModify {boneName=dyn_skirtflap_bk_l_MSTR}
    rotation=<rot 75 -40 30>
    translation {x=-10, z=5}
   }
   n5 = "Transform (Modify) Bone - Bone: dyn_skirtflap_bk_r_MSTR"(ComponentPose=@n4, Alpha=1, Translation="-10,0,5", Rotation="75,40,-30", Scale="1,1,1") {
    rotationMode=BMM_Additive
    rotationSpace=BCS_BoneSpace
    scaleMode=BMM_Replace
    scaleSpace=BCS_BoneSpace
    translationMode=BMM_Additive
    translationSpace=BCS_BoneSpace
    alphaBoolBlend {blendOption=Linear}
    alphaScaleBiasClamp {interpSpeedDecreasing=4}
    boneToModify {boneName=dyn_skirtflap_bk_r_MSTR}
    rotation=<rot 75 40 -30>
    translation {x=-10, z=5}
   }
   n6 = "Transform (Modify) Bone - Bone: dyn_skirtflap_r_MSTR"(ComponentPose=@n5, Alpha=1, Translation="0,0,0", Rotation="11.313,-116.602,-166.656", Scale="1,1,1") {
    rotationMode=BMM_Additive
    rotationSpace=BCS_BoneSpace
    scaleMode=BMM_Replace
    scaleSpace=BCS_BoneSpace
    translationMode=BMM_Additive
    translationSpace=BCS_BoneSpace
    alphaBoolBlend {blendOption=Linear}
    alphaScaleBiasClamp {interpSpeedDecreasing=4}
    boneToModify {boneName=dyn_skirtflap_r_MSTR}
    rotation=<rot 11.313 -116.602 -166.656>
   }
   n7 = "Transform (Modify) Bone - Bone: dyn_skirtflap_l_MSTR"(ComponentPose=@n6, Alpha=1, Translation="0,0,0", Rotation="80.000,150.000,-90.000", Scale="1,1,1") {
    rotationMode=BMM_Additive
    rotationSpace=BCS_BoneSpace
    scaleMode=BMM_Replace
    scaleSpace=BCS_BoneSpace
    translationMode=BMM_Additive
    translationSpace=BCS_BoneSpace
    alphaBoolBlend {blendOption=Linear}
    alphaScaleBiasClamp {interpSpeedDecreasing=4}
    boneToModify {boneName=dyn_skirtflap_l_MSTR}
    rotation=<rot 80 150 -90>
   }
   n8 = "Transform (Modify) Bone - Bone: dyn_skirtflap_bk_r_MSTR"(ComponentPose=@n7, Alpha=1, Translation="-10,-5,2.5", Rotation="-30.000,-150.000,-140.000", Scale="1,1,1") {
    rotationMode=BMM_Additive
    rotationSpace=BCS_BoneSpace
    scaleMode=BMM_Replace
    scaleSpace=BCS_BoneSpace
    translationMode=BMM_Additive
    translationSpace=BCS_BoneSpace
    alphaBoolBlend {blendOption=Linear}
    alphaScaleBiasClamp {interpSpeedDecreasing=4}
    boneToModify {boneName=dyn_skirtflap_bk_r_MSTR}
    rotation=<rot -30 -150 -140>
    translation=<-10 -5 2.5>
   }
   n9 = "Transform (Modify) Bone - Bone: dyn_skirtflap_bk_l_MSTR"(ComponentPose=@n8, Alpha=1, Translation="-10,5,2.5", Rotation="19.921,144.681,-166.817", Scale="1,1,1") {
    rotationMode=BMM_Additive
    rotationSpace=BCS_BoneSpace
    scaleMode=BMM_Replace
    scaleSpace=BCS_BoneSpace
    translationMode=BMM_Additive
    translationSpace=BCS_BoneSpace
    alphaBoolBlend {blendOption=Linear}
    alphaScaleBiasClamp {interpSpeedDecreasing=4}
    boneToModify {boneName=dyn_skirtflap_bk_l_MSTR}
    rotation=<rot 19.921 144.681 -166.817>
    translation=<-10 5 2.5>
   }
   n10 = RigidBody!(ComponentPose=@n9, Alpha, OverrideWorldGravity:Vector="0,0,0", ExternalForce) -> (Pose) {
    bOverrideWorldGravity=true
    overridePhysicsAsset=&/F_MED_SlumpAfter_Awe_Physics
    simulationSpace=BaseBoneSpace
    alphaBoolBlend {blendOption=Linear}
    baseBoneRef {boneName=pelvis}
    simSpaceSettings {
     maxAngularAcceleration=4
     maxAngularVelocity=8
     maxLinearAcceleration=4000
     maxLinearVelocity=800
     velocityScaleZ=0.5
     worldAlpha=1
     externalLinearDragV {y=2, z=2}
    }
   }
   n11 = "Transform (Modify) Bone - Bone: dyn_skirtflap_r_MSTR"(ComponentPose=@n10, Alpha=1, Translation="0,0,0", Rotation="0,0,0", Scale="0.4,1,1") {
    rotationMode=BMM_Additive
    rotationSpace=BCS_BoneSpace
    scaleMode=BMM_Replace
    scaleSpace=BCS_BoneSpace
    translationMode=BMM_Additive
    translationSpace=BCS_BoneSpace
    alphaBoolBlend {blendOption=Linear}
    alphaScaleBiasClamp {interpSpeedDecreasing=4}
    boneToModify {boneName=dyn_skirtflap_r_MSTR}
    scale {x=0.4}
   }
   n12 = "Transform (Modify) Bone - Bone: dyn_skirtflap_l_MSTR"(ComponentPose=@n11, Alpha=1, Translation="0,0,0", Rotation="0,0,0", Scale="0.4,1,1") {
    rotationMode=BMM_Additive
    rotationSpace=BCS_BoneSpace
    scaleMode=BMM_Replace
    scaleSpace=BCS_BoneSpace
    translationMode=BMM_Additive
    translationSpace=BCS_BoneSpace
    alphaBoolBlend {blendOption=Linear}
    alphaScaleBiasClamp {interpSpeedDecreasing=4}
    boneToModify {boneName=dyn_skirtflap_l_MSTR}
    scale {x=0.4}
   }
   n13 = "Transform (Modify) Bone - Bone: dyn_skirtflap_bk_r_MSTR"(ComponentPose=@n12, Alpha=1, Translation="0,0,0", Rotation="0,0,0", Scale="0.4,1,1") {
    rotationMode=BMM_Additive
    rotationSpace=BCS_BoneSpace
    scaleMode=BMM_Replace
    scaleSpace=BCS_BoneSpace
    translationMode=BMM_Additive
    translationSpace=BCS_BoneSpace
    alphaBoolBlend {blendOption=Linear}
    alphaScaleBiasClamp {interpSpeedDecreasing=4}
    boneToModify {boneName=dyn_skirtflap_bk_r_MSTR}
    scale {x=0.4}
   }
   n14 = "Transform (Modify) Bone - Bone: dyn_skirtflap_bk_l_MSTR"(ComponentPose=@n13, Alpha=1, Translation="0,0,0", Rotation="0,0,0", Scale="0.4,1,1") {
    rotationMode=BMM_Additive
    rotationSpace=BCS_BoneSpace
    scaleMode=BMM_Replace
    scaleSpace=BCS_BoneSpace
    translationMode=BMM_Additive
    translationSpace=BCS_BoneSpace
    alphaBoolBlend {blendOption=Linear}
    alphaScaleBiasClamp {interpSpeedDecreasing=4}
    boneToModify {boneName=dyn_skirtflap_bk_l_MSTR}
    scale {x=0.4}
   }
   Output Pose(Result=Component To Local(ComponentPose=@n14)) {layerGroup=None}
  ]}
 EventGraph {nodes [
   Event BlueprintUpdateAnimation()
   > Sequence()
   > Set GravityOverride(GravityOverride="0,-120,-470")
   > Set Is Knee Sliding Alpha(Is Knee Sliding Alpha:double="To Float (Boolean)"(InBool=Get bIsKneeSliding())) -> (Output_Get:double)
   > Set Is Baseball Sliding(Is Baseball Sliding:double="To Float (Boolean)"(InBool=Get bIsBaseballSliding())) -> (Output_Get:double)
   > Set Is In Vehicle Alpha(Is In Vehicle Alpha:double="To Float (Boolean)"(InBool=Get bIsInVehicle() -> (bIsInVehicle:bool))) -> (Output_Get:double)
  ]}
]
variables [
 GravityOverride {type=Vector}
 Is Baseball Sliding {type=number}
 Is In Vehicle Alpha {type=number}
 Is Knee Sliding Alpha {type=number}
]
