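# Physics asset for the SlumpAfter_Awe female character's FaceAcc_Cape part, covering secondary motion of left and right ponytails and four bows.
# Each five-bone ponytail and two-bone bow is a capsule chain constrained back to a tiny MSTR root body.
# Joints use tight swing and twist limits, with some bows allowed wider swing.
# Unconstrained capsules on head, neck, spine, clavicles and upper arms form body collision volumes.
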PhysicsAsset /BRCosmetics/Characters/Player/Female/Medium/Bodies/F_MED_SlumpAfter_Awe/Meshes/Parts/F_MED_SlumpAfter_Awe_FaceAcc_Cape_Physics
bodies {
 clavicle_l [{shapeType=Capsule, length=12, radius=6, shapeName=clavicle_l_capsule, center=<3.5 -1 -1>, rotation=<rot 80 -0.0003 -0.0003>}]
 clavicle_r [{shapeType=Capsule, length=12, radius=6, shapeName=clavicle_r_capsule, center=<-3.5 1 1>, rotation=<rot 80 -0.0003 -0.0004>}]
 dyn_bow_bk_l_01 [{shapeType=Capsule, length=3, radius=1.5, shapeName=dyn_bow_bk_l_01_capsule, center=<2 0 0>, rotation=<rot 0 0 90>}]
 dyn_bow_bk_l_02 [{shapeType=Capsule, length=3, radius=1.5, shapeName=dyn_bow_bk_l_02_capsule, center=<2 0 0>, rotation=<rot 0 0 90>}]
 dyn_bow_bk_l_MSTR [{shapeType=Capsule, length=0.505, radius=0.505, shapeName=dyn_bow_bk_l_MSTR_capsule, rotation=<rot 0 -90 0>}]
 dyn_bow_bk_r_01 [{shapeType=Capsule, length=3, radius=1.5, shapeName=dyn_bow_bk_r_01_capsule, center=<2 0 0>, rotation=<rot 0 0 90>}]
 dyn_bow_bk_r_02 [{shapeType=Capsule, length=3, radius=1.5, shapeName=dyn_bow_bk_r_02_capsule, center=<2 0 0>, rotation=<rot 0 0 90>}]
 dyn_bow_bk_r_MSTR [{shapeType=Capsule, length=0.505, radius=0.505, shapeName=dyn_bow_bk_r_MSTR_capsule, rotation=<rot 0 -90 0>}]
 dyn_bow_ft_l_01 [{shapeType=Capsule, length=3, radius=1.5, shapeName=dyn_bow_ft_l_01_capsule, center=<2 0 0>, rotation=<rot 0 0 90>}]
 dyn_bow_ft_l_02 [{shapeType=Capsule, length=3, radius=1.5, shapeName=dyn_bow_ft_l_02_capsule, center=<2 0 0>, rotation=<rot 0 0 90>}]
 dyn_bow_ft_l_MSTR [{shapeType=Capsule, length=0.505, radius=0.505, shapeName=dyn_bow_ft_l_MSTR_capsule, rotation=<rot 0 -90 0>}]
 dyn_bow_ft_r_01 [{shapeType=Capsule, length=3, radius=1.5, shapeName=dyn_bow_ft_r_01_capsule, center=<2 0 0>, rotation=<rot 0 0 90>}]
 dyn_bow_ft_r_02 [{shapeType=Capsule, length=3, radius=1.5, shapeName=dyn_bow_ft_r_02_capsule, center=<2 0 0>, rotation=<rot 0 0 90>}]
 dyn_bow_ft_r_MSTR [{shapeType=Capsule, length=0.505, radius=0.505, shapeName=dyn_bow_ft_r_MSTR_capsule, rotation=<rot 0 -90 0>}]
 dyn_ponytail_l_01 [{shapeType=Capsule, length=6, radius=2.5, shapeName=dyn_ponytail_l_01_capsule, center=<6 0 0>, rotation=<rot 0 -90 90>}]
 dyn_ponytail_l_02 [{shapeType=Capsule, length=5, radius=4, shapeName=dyn_ponytail_l_02_capsule, center=<6 0 1.5>, rotation=<rot 0 -90 90>}]
 dyn_ponytail_l_03 [{shapeType=Capsule, length=5, radius=4.5, shapeName=dyn_ponytail_l_03_capsule, center=<6 1 2>, rotation=<rot 0 -90 90>}]
 dyn_ponytail_l_04 [{shapeType=Capsule, length=5, radius=4.5, shapeName=dyn_ponytail_l_04_capsule, center=<6 1 2>, rotation=<rot 0 -90 90>}]
 dyn_ponytail_l_05 [{shapeType=Capsule, length=5, radius=4.5, shapeName=dyn_ponytail_l_05_capsule, center=<6 0 2>, rotation=<rot 0 -89.9999 100.0001>}]
 dyn_ponytail_l_MSTR [{shapeType=Capsule, length=0.505, radius=0.505, shapeName=dyn_ponytail_l_MSTR_capsule, rotation=<rot 0 -90 0>}]
 dyn_ponytail_r_01 [{shapeType=Capsule, length=6, radius=2.5, shapeName=dyn_ponytail_r_01_capsule, center=<6 0 0>, rotation=<rot 0 -90 90>}]
 dyn_ponytail_r_02 [{shapeType=Capsule, length=5, radius=4, shapeName=dyn_ponytail_r_02_capsule, center=<6 0 1.5>, rotation=<rot 0 -90 90>}]
 dyn_ponytail_r_03 [{shapeType=Capsule, length=5, radius=4.5, shapeName=dyn_ponytail_r_03_capsule, center=<6 -1 2>, rotation=<rot 0 -90 90>}]
 dyn_ponytail_r_04 [{shapeType=Capsule, length=5, radius=4.5, shapeName=dyn_ponytail_r_04_capsule, center=<6 -1 2>, rotation=<rot 0 -90 90>}]
 dyn_ponytail_r_05 [{shapeType=Capsule, length=5, radius=4.5, shapeName=dyn_ponytail_r_05_capsule, center=<6 0 2>, rotation=<rot 0 -90.0001 100.0001>}]
 dyn_ponytail_r_MSTR [{shapeType=Capsule, length=0.505, radius=0.505, shapeName=dyn_ponytail_r_MSTR_capsule, rotation=<rot 0 -90 0>}]
 head [{shapeType=Capsule, length=1, radius=11, shapeName=head_capsule, center=<7.5005 1.4268 0>, rotation=<rot 90 0 0>}]
 neck_01 [{shapeType=Capsule, length=8, radius=4.5, shapeName=neck_01_capsule, center=<3.6683 -0.2876 0>, rotation=<rot 90 0 0>}]
 spine_03 [{shapeType=Capsule, length=15, radius=10, shapeName=spine_04_capsule, center=<0.229 2 0>, rotation=<rot 90 0 10.0001>}]
 spine_05 [
  {shapeType=Capsule, length=6, radius=11, shapeName=spine_05_capsule, center=<-2.5963 0.209 0>}
  {shapeType=Capsule, length=6, radius=11, shapeName=spine_05_capsule, center=<-5.7241 1.6209 0>}
  {shapeType=Capsule, length=15, radius=12, shapeName=spine_05_capsule, center=<-0.5072 -20.6786 0>, rotation=<rot 90 9.9999 0>}
 ]
 upperarm_l [{shapeType=Capsule, length=15, radius=8, shapeName=upperarm_l_capsule, center=<6 -0.5 0>, rotation=<rot 90 0 0>}]
 upperarm_r [{shapeType=Capsule, length=15, radius=8, shapeName=upperarm_r_capsule, center=<-6 0.5 0>, rotation=<rot 90 0 0>}]
}
constraints [
 {bone1Name=dyn_ponytail_r_05, bone2Name=dyn_ponytail_r_04, swing1LimitDegrees=15, swing2LimitDegrees=20, twistLimitDegrees=5}
 {bone1Name=dyn_bow_bk_l_01, bone2Name=dyn_bow_bk_l_MSTR, swing1LimitDegrees=5, swing2LimitDegrees=25, twistLimitDegrees=1}
 {bone1Name=dyn_bow_bk_l_02, bone2Name=dyn_bow_bk_l_01, swing1LimitDegrees=5, swing2LimitDegrees=25, twistLimitDegrees=1}
 {bone1Name=dyn_bow_bk_r_01, bone2Name=dyn_bow_bk_r_MSTR, swing1LimitDegrees=5, swing2LimitDegrees=45, twistLimitDegrees=1}
 {bone1Name=dyn_bow_bk_r_02, bone2Name=dyn_bow_bk_r_01, swing1LimitDegrees=5, swing2LimitDegrees=45, twistLimitDegrees=1}
 {bone1Name=dyn_bow_ft_l_01, bone2Name=dyn_bow_ft_l_MSTR, swing1LimitDegrees=5, swing2LimitDegrees=45, twistLimitDegrees=1}
 {bone1Name=dyn_bow_ft_l_02, bone2Name=dyn_bow_ft_l_01, swing1LimitDegrees=5, swing2LimitDegrees=45, twistLimitDegrees=1}
 {bone1Name=dyn_bow_ft_r_01, bone2Name=dyn_bow_ft_r_MSTR, swing1LimitDegrees=5, swing2LimitDegrees=25, twistLimitDegrees=1}
 {bone1Name=dyn_bow_ft_r_02, bone2Name=dyn_bow_ft_r_01, swing1LimitDegrees=5, swing2LimitDegrees=25, twistLimitDegrees=1}
 {bone1Name=dyn_ponytail_l_01, bone2Name=dyn_ponytail_l_MSTR, swing1LimitDegrees=15, swing2LimitDegrees=20, twistLimitDegrees=1}
 {bone1Name=dyn_ponytail_l_02, bone2Name=dyn_ponytail_l_01, swing1LimitDegrees=15, swing2LimitDegrees=20, twistLimitDegrees=5}
 {bone1Name=dyn_ponytail_l_03, bone2Name=dyn_ponytail_l_02, swing1LimitDegrees=15, swing2LimitDegrees=20, twistLimitDegrees=5}
 {bone1Name=dyn_ponytail_l_04, bone2Name=dyn_ponytail_l_03, swing1LimitDegrees=15, swing2LimitDegrees=20, twistLimitDegrees=5}
 {bone1Name=dyn_ponytail_l_05, bone2Name=dyn_ponytail_l_04, swing1LimitDegrees=15, swing2LimitDegrees=20, twistLimitDegrees=5}
 {bone1Name=dyn_ponytail_r_01, bone2Name=dyn_ponytail_r_MSTR, swing1LimitDegrees=15, swing2LimitDegrees=20, twistLimitDegrees=1}
 {bone1Name=dyn_ponytail_r_02, bone2Name=dyn_ponytail_r_01, swing1LimitDegrees=15, swing2LimitDegrees=20, twistLimitDegrees=5}
 {bone1Name=dyn_ponytail_r_03, bone2Name=dyn_ponytail_r_02, swing1LimitDegrees=15, swing2LimitDegrees=20, twistLimitDegrees=5}
 {bone1Name=dyn_ponytail_r_04, bone2Name=dyn_ponytail_r_03, swing1LimitDegrees=15, swing2LimitDegrees=20, twistLimitDegrees=5}
]
